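# M_MiniMapPapaya is a User Interface domain material for the minimap of the Papaya map.
# Its only node is a texture sample of the MiniMapPapaya texture, with no parameters, math or other inputs.
# The graph lists no link from that sample to a material output.
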Material /PapayaGameplay/Athena/Apollo/Maps/Special/Papaya/UI/M_MiniMapPapaya
graph {
 properties {materialDomain=MD_UI, pixelDepthOffsetMode=PDOM_Legacy, usedShadingModels=Default Lit}
 nodes [TextureSample() {texture=/Game/Athena/Apollo/Maps/Special/Papaya/UI/MiniMapPapaya}]
}
